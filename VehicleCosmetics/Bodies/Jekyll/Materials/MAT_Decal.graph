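# Masked, Default Lit decal material for the Jekyll vehicle body cosmetic, enabled for skeletal meshes and mesh deformers.
# The red channel of a grayscale decal texture scales a near-black constant color, and a separate decal texture is sampled as a normal map.
# The graph is fixed, with no exposed parameters, and its listing does not show which material outputs these nodes feed.
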Material /VehicleCosmetics/Bodies/Jekyll/Materials/MAT_Decal
graph {
 properties {
  bAutomaticallySetUsageInEditor=false
  bUsedWithMeshDeformer=true
  bUsedWithSkeletalMesh=true
  blendMode=BLEND_Masked
  usedShadingModels=Default Lit
 }
 nodes [
  Multiply(
   A=Constant3Vector() {constant=<rgba 0.02 0.02 0.02 0>}
   B=TextureSample() {samplerType=SAMPLERTYPE_LinearGrayscale, texture=/VehicleCosmetics/Bodies/Jekyll/Textures/T_Jekyll_Decal_A}.R
  )
  TextureSample() {samplerType=SAMPLERTYPE_Normal, texture=/VehicleCosmetics/Bodies/Jekyll/Textures/T_Jekyll_Decal}
 ]
}
